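# Niagara jet exhaust and airflow effect for the Space Warrior (male) glider.
# Flame meshes, fire sprites and glow cards follow skeletal mesh sockets for one big and two small thrusters, with airflow meshes and motion lines on further sockets.
# Three FN_Integrator values act as smoothed drivers, scaling thrust and crossfading downward airflow into backward airflow.
# User forward/right dot, deployed and front-end inputs adjust individual jets, spawning and scale.
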NiagaraSystem /BRCosmetics/Gadgets/Assets/VinderTech_GliderChute/Glider_Space_Warrior_Male/FX/NS_Glider_SpaceWarrior_M_Jet
dependencies {
 usedDataInterfaces [
  &NiagaraDataInterfaceVectorCurve
  &NiagaraDataInterfaceCurve
  &NiagaraDataInterfaceVector2DCurve
 ]
 usedDynamicInputs []
 usedModules [&SystemState, &FN_Integrator]
}
emitters {
 Airflow_Back {
  emitterUpdateScript [
   &EmitterState(Life Cycle Mode=::NewEnumerator1, UseLoopDelay=true)
   &SpawnBurst_Instantaneous(Spawn Count=16, Use Spawn Probability=false)
  ]
  particleSpawnScript [
   /Niagara/Modules/Spawn/Initialization/InitializeParticle(
    Color=<rgba 1 1 1 0.3>
    Lifetime=5
    Module.Write Color=true
    Module.Write Lifetime=true
    Module.Write Mass=true
    Module.Write Material Random=true
    Module.Write Position=false
    Module.Write Scale=true
    Module.Write SpriteRotation=false
    Module.Write SpriteSize=false
    Position=<0 0 10>
   )
   &FN_SetSpawnGroup()
   &SkeletalMeshLocation(
    Apply Sampled Orientation to Mesh Particles=true
    Bone / Socket Index=&Add_Integer(A=4, B=$Particles.SpawnGroup)
    Inherit Velocity=false
    MaskSpawnGroup=false
    Mesh Sampling Type=::NewEnumerator5
    Sampled Orientation Coordinate Space=::Local
    Sampled Position Coordinate Space=::Local
    Sampled Position Offset=<-10 0 0>
    Sampled Position Offset Coordinate Space:ENiagaraMeshTransforms=::NewEnumerator2
    Skeletal Mesh=$User.SKMesh
    Socket Sampling Mode=::NewEnumerator5
   ) {module=SkeletalMeshLocation001}
   &InitialMeshOrientation(
    Orientation Axis=<0 0 0>
    Orientation Coordinate Space=::Local
    Orientation Vector=<0 0 1>
    Rotation=<0 0.1 0>
    Rotation Coordinate Space=::NewEnumerator2
    Use Orientation Vector=false
    Use Rotation Vector=true
   )
   /SetVariables(
    Particles.MeshScale_02:Vector3f=<1 1 1>
    Particles.XOffset:NiagaraFloat=&RandomRangeFloat(Minimum=0, Maximum=1)
    Particles.XOffset_02:NiagaraFloat=&RandomRangeFloat(Minimum=0, Maximum=1)
   ) {isSetParameters=true, module=SetVariables_#1}
  ]
  particleUpdateScript [
   &ParticleState()
   &SkeletalMeshLocation(
    Apply Sampled Orientation to Mesh Particles=true
    Bone / Socket Index=&Add_Integer(A=4, B=$Particles.SpawnGroup)
    Inherit Velocity=false
    MaskSpawnGroup=false
    Mesh Sampling Type=::NewEnumerator5
    Sampled Orientation Coordinate Space=::Local
    Sampled Position Coordinate Space=::Local
    Sampled Position Offset=&Lerp_Vector(
     A=<0 -2.5 0>
     B=<0 2 0>
     Alpha=&MakeFloatFromInt(Int=&ModuloInt(A=$Particles.Initial.SpawnGroup, B=2))
    )
    Sampled Position Offset Coordinate Space:ENiagaraMeshTransforms=::NewEnumerator3
    Skeletal Mesh=$User.SKMesh
    Socket Sampling Mode=::NewEnumerator5
   )
   &DynamicMaterialParameters(
    Index 0 Param 1=1
    Index 0 Param 2=0.04
    Index 0 Param 3=20
    Index 0 Param 4=1
    Index 1 Param 1=&Add_Float(
     A=$Particles.Initial.XOffset
     B=&FloatFromCurve(FloatCurve=DataInterface() {propertyValues {Curve {keys [<at 0 0>,<at 1 1>]}}}, Scale Curve=2.5)
    )
    Index 1 Param 2=&RandomRangeFloat(Minimum=0, Maximum=1)
    Index 2 Param 1=&Add_Float(
     A=$Particles.Initial.XOffset_02
     B=&FloatFromCurve(FloatCurve=DataInterface() {propertyValues {Curve {keys [<at 0 0>,<at 1 1>]}}}, Scale Curve=2.5)
    )
    Index 2 Param 2=&RandomRangeFloat(Minimum=0, Maximum=1)
    Module.Param0WriteEnabled=true
    Module.Param0WriteEnabled1=true
    Module.Param0WriteEnabled2=true
    Module.Param1WriteEnabled=false
    Module.Param1WriteEnabled1=true
    Module.Param1WriteEnabled2=true
    Module.Param2WriteEnabled=false
    Module.Param2WriteEnabled1=false
    Module.Param2WriteEnabled2=true
    Module.Param3WriteEnabled=true
    Module.Param3WriteEnabled1=false
    Module.Param3WriteEnabled2=true
    Write Parameter Index 1=true
    Write Parameter Index 2=true
   ) {module=DynamicMaterialParameters001}
   &ScaleColor(
    Scale Alpha=&Lerp_Float(
     A=0
     B=0.2
     Alpha=/Niagara/DynamicInputs/Math/ReturnLargestFloat(
      A:NiagaraFloat=$System.FN_Integrator.NormalizedValue
      B:NiagaraFloat=&ReturnLargestFloat(A:NiagaraFloat=$System.FN_Integrator002.NormalizedValue, B:NiagaraFloat=$System.FN_Integrator001.NormalizedValue)
     )
    )
    ScaleA=true
    ScaleRGB=false
   )
   /SetVariables(Particles.MeshScale_02:Vector3f=&Multiply_Vector(A=<1 0.25 1>, B=$Particles.Scale)) {isSetParameters=true, module=SetVariables_#2}
  ]
  rendererClasses [&NiagaraMeshRendererProperties]
  renderers [
   {rendererClass=&NiagaraMeshRendererProperties, rendererIndex=0}
   {rendererClass=&NiagaraMeshRendererProperties, rendererIndex=1}
  ]
 }
 Airflow_Down {
  emitterUpdateScript [
   &EmitterState(Life Cycle Mode=::NewEnumerator1, UseLoopDelay=true)
   &SpawnBurst_Instantaneous(Spawn Count=16, Use Spawn Probability=false)
  ]
  particleSpawnScript [
   &InitializeParticle(
    Color=<rgba 1 1 1 0.3>
    Lifetime=5
    Module.Write Color=true
    Module.Write Lifetime=true
    Module.Write Mass=true
    Module.Write Material Random=true
    Module.Write Position=false
    Module.Write Scale=true
    Module.Write SpriteRotation=false
    Module.Write SpriteSize=false
    Position=<0 0 10>
   )
   &FN_SetSpawnGroup()
   &SkeletalMeshLocation(
    Apply Sampled Orientation to Mesh Particles=true
    Bone / Socket Index=&Add_Integer(A=4, B=$Particles.SpawnGroup)
    Inherit Velocity=false
    MaskSpawnGroup=false
    Mesh Sampling Type=::NewEnumerator5
    Sampled Orientation Coordinate Space=::Local
    Sampled Position Coordinate Space=::Local
    Sampled Position Offset=<-10 0 0>
    Sampled Position Offset Coordinate Space:ENiagaraMeshTransforms=::NewEnumerator2
    Skeletal Mesh=$User.SKMesh
    Socket Sampling Mode=::NewEnumerator5
   ) {module=SkeletalMeshLocation001}
   &InitialMeshOrientation(
    Orientation Axis=<0 0 0>
    Orientation Coordinate Space=::Local
    Orientation Vector=<0 0 1>
    Rotation=<0 0.1 0>
    Rotation Coordinate Space=::NewEnumerator2
    Use Orientation Vector=false
    Use Rotation Vector=true
   )
   /SetVariables(
    Particles.MeshScale_02:Vector3f=<1 1 1>
    Particles.XOffset:NiagaraFloat=&RandomRangeFloat(Minimum=0, Maximum=1)
    Particles.XOffset_02:NiagaraFloat=&RandomRangeFloat(Minimum=0, Maximum=1)
   ) {isSetParameters=true, module=SetVariables_#1}
  ]
  particleUpdateScript [
   &ParticleState()
   &SkeletalMeshLocation(
    Apply Sampled Orientation to Mesh Particles=true
    Bone / Socket Index=&Add_Integer(A=4, B=$Particles.SpawnGroup)
    Inherit Velocity=false
    MaskSpawnGroup=false
    Mesh Sampling Type=::NewEnumerator5
    Sampled Orientation Coordinate Space=::Local
    Sampled Position Coordinate Space=::Local
    Sampled Position Offset=&Lerp_Vector(
     A=<0 -2.5 0>
     B=<0 2 0>
     Alpha=&MakeFloatFromInt(Int=&ModuloInt(A=$Particles.Initial.SpawnGroup, B=2))
    )
    Sampled Position Offset Coordinate Space:ENiagaraMeshTransforms=::NewEnumerator3
    Skeletal Mesh=$User.SKMesh
    Socket Sampling Mode=::NewEnumerator5
   )
   &DynamicMaterialParameters(
    Index 0 Param 1=1
    Index 0 Param 2=0.04
    Index 0 Param 3=20
    Index 0 Param 4=1
    Index 1 Param 1=&Add_Float(
     A=$Particles.Initial.XOffset
     B=&FloatFromCurve(FloatCurve=DataInterface() {propertyValues {Curve {keys [<at 0 0>,<at 1 1>]}}}, Scale Curve=2.5)
    )
    Index 1 Param 2=&RandomRangeFloat(Minimum=0, Maximum=1)
    Index 2 Param 1=&Add_Float(
     A=$Particles.Initial.XOffset_02
     B=&FloatFromCurve(FloatCurve=DataInterface() {propertyValues {Curve {keys [<at 0 0>,<at 1 1>]}}}, Scale Curve=2.5)
    )
    Index 2 Param 2=&RandomRangeFloat(Minimum=0, Maximum=1)
    Module.Param0WriteEnabled=true
    Module.Param0WriteEnabled1=true
    Module.Param0WriteEnabled2=true
    Module.Param1WriteEnabled=false
    Module.Param1WriteEnabled1=true
    Module.Param1WriteEnabled2=true
    Module.Param2WriteEnabled=false
    Module.Param2WriteEnabled1=false
    Module.Param2WriteEnabled2=true
    Module.Param3WriteEnabled=true
    Module.Param3WriteEnabled1=false
    Module.Param3WriteEnabled2=true
    Write Parameter Index 1=true
    Write Parameter Index 2=true
   ) {module=DynamicMaterialParameters001}
   &ScaleColor(
    Scale Alpha=&Lerp_Float(
     A=0.25
     B=0
     Alpha=&ReturnLargestFloat(
      A:NiagaraFloat=$System.FN_Integrator.NormalizedValue
      B:NiagaraFloat=&ReturnLargestFloat(A:NiagaraFloat=$System.FN_Integrator002.NormalizedValue, B:NiagaraFloat=$System.FN_Integrator001.NormalizedValue)
     )
    )
    ScaleA=true
    ScaleRGB=false
   )
   /SetVariables(Particles.MeshScale_02:Vector3f=&Multiply_Vector(A=<1 0.25 1>, B=$Particles.Scale)) {isSetParameters=true, module=SetVariables_#2}
  ]
  rendererClasses [&NiagaraMeshRendererProperties]
  renderers [
   {rendererClass=&NiagaraMeshRendererProperties, rendererIndex=0}
   {rendererClass=&NiagaraMeshRendererProperties, rendererIndex=1}
  ]
 }
 Airflow_MotionLines {
  emitterUpdateScript [
   &EmitterState()
   &SpawnRate(SpawnRate=40, Use Spawn Probability=false)
  ]
  particleSpawnScript [
   &InitializeParticle(
    Lifetime=&RandomRangeFloat(Minimum=0.2, Maximum=0.3, Evaluation Type=::NewEnumerator1)
    Module.Write Color=true
    Module.Write Lifetime=true
    Module.Write Mass=true
    Module.Write Material Random=true
    Module.Write Position=false
    Module.Write Scale=false
    Module.Write SpriteRotation=true
    Module.Write SpriteSize=true
    Sprite Rotation=&NormalizedAngleToDegrees(NormalizedAngle=&RandomRangeFloat(Minimum=0, Maximum=1, Evaluation Type=::NewEnumerator1))
    Sprite Size=&Multiply_Vector2(A=&RandomRangeVector2D(Minimum=<0.5 1.5>, Maximum=<0.75 2>), B=<2 8>)
   )
   &FN_SetSpawnGroup()
   &Color(Color=&MakeLinearColorFromVectorAndFloat("Vector (RGB)"=<1 1 1>, "Float (Alpha)"=&ClampFloat(Float=0.2)))
   &SkeletalMeshLocation(
    Bone / Socket Index=&Add_Integer(A=$Particles.SpawnGroup, B=4)
    Inherit Velocity=false
    MaskSpawnGroup=false
    Mesh Sampling Type=::NewEnumerator5
    Sampled Orientation Coordinate Space=::Local
    Sampled Position Coordinate Space=::Local
    Sampled Position Offset=&RandomRangeVector(Minimum=<-4 -4 -4>, Maximum=<4 4 4>)
    Sampled Position Offset Coordinate Space:ENiagaraMeshTransforms=::NewEnumerator2
    Sampled Velocity Coordinate Space:ENiagaraCoordinateSpace=::Local
    Skeletal Mesh=$User.SKMesh
    Socket Sampling Mode=::NewEnumerator5
   )
   &AddVelocity(Velocity=&RotateVectorByQuaternion(
     Vector To Rotate=&Multiply_Vector(
      A=&RandomRangeVector(Minimum=<1 0.2 0.35>, Maximum=<1 0 0.45>)
      B=&MakeVector(X=350, Y=0, Z=350)
     )
     Quaternion=$Output.SkeletalMeshLocation.SampledRotation
    ))
  ]
  particleUpdateScript [
   &ParticleState()
   &ScaleColor(
    Scale Alpha=&FloatFromCurve(FloatCurve=DataInterface() {propertyValues {Curve {keys [
         <at 0 0 RCIM_Cubic>
         <at 0.2 1>
         {arriveTangent=-0.874, interpMode=RCIM_Cubic, leaveTangent=-0.874, time=0.5987, value=0.9935}
         {arriveTangent=-2.4759, interpMode=RCIM_Cubic, leaveTangent=-2.4759, time=0.8162, value=0.4615}
         <at 1 0>
        ]}}})
    Scale RGB=&VectorFromCurve(VectorCurve=DataInterface() {propertyValues {
       XCurve {keys [{arriveTangent=-861.5734, interpMode=RCIM_Cubic, time=0, value=250}, <at 0.2668 150 RCIM_Cubic>]}
       YCurve {keys [{arriveTangent=-66.82, interpMode=RCIM_Cubic, time=0, value=25}, <at 0.2668 5 RCIM_Cubic>]}
       ZCurve {keys [{arriveTangent=-0.0001, interpMode=RCIM_Cubic, time=0, value=1}, <at 0.2668 0 RCIM_Cubic>]}
      }})
    ScaleA=true
    ScaleRGB=false
   )
   &ScaleColor(
    Scale Alpha=&Lerp_Float(
     A=0
     B=1
     Alpha=&ReturnLargestFloat(
      A:NiagaraFloat=$System.FN_Integrator.NormalizedValue
      B:NiagaraFloat=&ReturnLargestFloat(A:NiagaraFloat=$System.FN_Integrator002.NormalizedValue, B:NiagaraFloat=$System.FN_Integrator001.NormalizedValue)
     )
    )
    ScaleA=true
    ScaleRGB=false
   ) {module=ScaleColor001}
   &SolveForcesAndVelocity(Clamp Velocity=false, Limit Acceleration=false)
   &ScaleSpriteSizeBySpeed(Max Scale Factor=<1 6>)
  ]
  rendererClasses [&NiagaraSpriteRendererProperties]
  renderers [{rendererClass=&NiagaraSpriteRendererProperties}]
 }
 Airflow_MotionLines001 {
  emitterUpdateScript [
   &EmitterState()
   &SpawnRate(SpawnRate=40, Use Spawn Probability=false)
  ]
  particleSpawnScript [
   &InitializeParticle(
    Lifetime=&RandomRangeFloat(Minimum=0.2, Maximum=0.3, Evaluation Type=::NewEnumerator1)
    Module.Write Color=true
    Module.Write Lifetime=true
    Module.Write Mass=true
    Module.Write Material Random=true
    Module.Write Position=false
    Module.Write Scale=false
    Module.Write SpriteRotation=true
    Module.Write SpriteSize=true
    Sprite Rotation=&NormalizedAngleToDegrees(NormalizedAngle=&RandomRangeFloat(Minimum=0, Maximum=1, Evaluation Type=::NewEnumerator1))
    Sprite Size=&Multiply_Vector2(A=&RandomRangeVector2D(Minimum=<0.5 1.5>, Maximum=<0.75 2>), B=<2 8>)
   )
   &FN_SetSpawnGroup()
   &Color(Color=&MakeLinearColorFromVectorAndFloat("Vector (RGB)"=<1 1 1>, "Float (Alpha)"=&ClampFloat(Float=0.15)))
   &SkeletalMeshLocation(
    Bone / Socket Index=&Add_Integer(A=$Particles.SpawnGroup, B=4)
    Inherit Velocity=false
    MaskSpawnGroup=false
    Mesh Sampling Type=::NewEnumerator5
    Sampled Orientation Coordinate Space=::Local
    Sampled Position Coordinate Space=::Local
    Sampled Position Offset=&RandomRangeVector(Minimum=<-7 -7 0>, Maximum=<7 7 0>)
    Sampled Position Offset Coordinate Space:ENiagaraMeshTransforms=::NewEnumerator2
    Sampled Velocity Coordinate Space:ENiagaraCoordinateSpace=::Local
    Skeletal Mesh=$User.SKMesh
    Socket Sampling Mode=::NewEnumerator5
   )
   &AddVelocity(Velocity=&Multiply_Vector(A=&RandomRangeVector(Minimum=<0 100 -250>, Maximum=<0 150 -350>), B=<1 1 1>)) {module=AddVelocity001}
   &AddVelocityFromPoint(
    Module.Use Falloff Distance=false
    Origin Offset=$Output.SkeletalMeshLocation.SampledPosition
    Velocity Strength=&RandomRangeFloat(Minimum=25, Maximum=50)
   )
  ]
  particleUpdateScript [
   &ParticleState()
   &ScaleColor(
    Scale Alpha=&FloatFromCurve(FloatCurve=DataInterface() {propertyValues {Curve {keys [
         <at 0 0 RCIM_Cubic>
         <at 0.2 1>
         {arriveTangent=-0.874, interpMode=RCIM_Cubic, leaveTangent=-0.874, time=0.5987, value=0.9935}
         {arriveTangent=-2.4759, interpMode=RCIM_Cubic, leaveTangent=-2.4759, time=0.8162, value=0.4615}
         <at 1 0>
        ]}}})
    Scale RGB=&VectorFromCurve(VectorCurve=DataInterface() {propertyValues {
       XCurve {keys [{arriveTangent=-861.5734, interpMode=RCIM_Cubic, time=0, value=250}, <at 0.2668 150 RCIM_Cubic>]}
       YCurve {keys [{arriveTangent=-66.82, interpMode=RCIM_Cubic, time=0, value=25}, <at 0.2668 5 RCIM_Cubic>]}
       ZCurve {keys [{arriveTangent=-0.0001, interpMode=RCIM_Cubic, time=0, value=1}, <at 0.2668 0 RCIM_Cubic>]}
      }})
    ScaleA=true
    ScaleRGB=false
   )
   &ScaleColor(
    Scale Alpha=&Lerp_Float(
     A=1
     B=0
     Alpha=&ReturnLargestFloat(
      A:NiagaraFloat=$System.FN_Integrator.NormalizedValue
      B:NiagaraFloat=&ReturnLargestFloat(A:NiagaraFloat=$System.FN_Integrator002.NormalizedValue, B:NiagaraFloat=$System.FN_Integrator001.NormalizedValue)
     )
    )
    ScaleA=true
    ScaleRGB=false
   ) {module=ScaleColor001}
   &SolveForcesAndVelocity(Clamp Velocity=false, Limit Acceleration=false)
   &ScaleSpriteSizeBySpeed(Max Scale Factor=<1 6>)
  ]
  rendererClasses [&NiagaraSpriteRendererProperties]
  renderers [{rendererClass=&NiagaraSpriteRendererProperties}]
 }
 FireBig {
  emitterUpdateScript [
   &EmitterState()
   &SpawnRate(
    SpawnRate=&FloatFromCurve(
     FloatCurve=DataInterface() {propertyValues {Curve {keys [
         <at 0 1>
         {arriveTangent=0.0423, interpMode=RCIM_Cubic, leaveTangent=0.0423, time=0.153, value=1.1762}
         {arriveTangent=-0.787, interpMode=RCIM_Cubic, leaveTangent=-0.787, time=0.2441, value=1.0103}
         <at 0.3053 1.0563>
         {arriveTangent=-0.5819, interpMode=RCIM_Cubic, leaveTangent=-0.5819, time=0.3571, value=1.0822}
         {arriveTangent=-0.9728, interpMode=RCIM_Cubic, leaveTangent=-0.9728, time=0.4934, value=0.9469}
         <at 0.5083 0.9351>
         {arriveTangent=1.0204, interpMode=RCIM_Cubic, leaveTangent=1.0204, time=0.6052, value=0.9718}
         {arriveTangent=0.0715, interpMode=RCIM_Cubic, leaveTangent=0.0715, time=0.7799, value=1.2121}
         <at 1 1>
        ]}}}
     CurveIndex=$Emitter.NormalizedLoopAge
     Scale Curve=&Lerp_Float(
      A=0
      B=&Lerp_Float(A=0, B=20, Alpha=$System.FN_Integrator.NormalizedValue)
      Alpha=&MakeCustomFloatFromBool(Bool=$User.bisFullyDeployed, True Float=1, False Float=0)
     )
    )
    Use Spawn Probability=false
   )
  ]
  particleSpawnScript [
   &InitializeParticle(
    Lifetime=&RandomRangeFloat(Minimum=0.2, Maximum=0.3, Evaluation Type=::NewEnumerator1)
    Module.Write Color=true
    Module.Write Lifetime=true
    Module.Write Mass=true
    Module.Write Material Random=true
    Module.Write Position=false
    Module.Write Scale=false
    Module.Write SpriteRotation=true
    Module.Write SpriteSize=true
    Sprite Rotation=&NormalizedAngleToDegrees(NormalizedAngle=&RandomRangeFloat(Minimum=0, Maximum=1, Evaluation Type=::NewEnumerator1))
    Sprite Size=&Vector2DFromFloat(Value=&RandomRangeFloat(Minimum=20, Maximum=25, Evaluation Type=::NewEnumerator1))
   )
   &Color(Color=&MakeLinearColorFromVectorAndFloat("Vector (RGB)"=<1 0.1 0>, "Float (Alpha)"=&ClampFloat(Float=0.1)))
   &SkeletalMeshLocation(
    Bone / Socket Index=2
    Inherit Velocity=false
    MaskSpawnGroup=false
    Mesh Sampling Type=::NewEnumerator5
    Sampled Orientation Coordinate Space=::Local
    Sampled Position Coordinate Space=::Local
    Sampled Velocity Coordinate Space:ENiagaraCoordinateSpace=::Local
    Skeletal Mesh=$User.SKMesh
    Socket Sampling Mode=::NewEnumerator5
   )
   &AddVelocity(Velocity=&RotateVectorByQuaternion(
     Vector To Rotate=&RandomRangeVector(Minimum=<40 0 0>, Maximum=<60 0 0>)
     Quaternion=$Output.SkeletalMeshLocation.SampledRotation
    ))
   &SphereLocation(
    MaskSpawnGroup=false
    Override Seed=false
    Sphere Radius=2.5
    Surface Only=false
    UseNonUniformScale=false
    UseOffset=false
    UseSphereOrientation=false
   )
  ]
  particleUpdateScript [
   &ParticleState()
   &Drag(Use Linear Drag=true, Use Rotational Drag=false)
   &SolveForcesAndVelocity(Clamp Velocity=false, Limit Acceleration=false)
   &ScaleSpriteSize(Scale Factor=&Vector2DFromFloat(Value=&FloatFromCurve(
      FloatCurve=DataInterface() {propertyValues {Curve {keys [{arriveTangent=1.4511, interpMode=RCIM_Cubic, time=-0.0013, value=0.1965}, <at 1 1.25>]}}}
      Scale Curve=&Lerp_Float(A=0.8, B=1.15, Alpha=$System.FN_Integrator.NormalizedValue)
     )))
   &ScaleColor(
    Scale Alpha=&FloatFromCurve(FloatCurve=DataInterface() {propertyValues {Curve {keys [
         <at 0 1 RCIM_Cubic>
         {arriveTangent=-0.573, interpMode=RCIM_Cubic, leaveTangent=-0.573, time=0.5987, value=0.9935}
         {arriveTangent=-2.4759, interpMode=RCIM_Cubic, leaveTangent=-2.4759, time=0.856, value=0.5095}
         <at 1 0>
        ]}}})
    Scale RGB=&VectorFromCurve(VectorCurve=DataInterface() {propertyValues {
       XCurve {keys [{arriveTangent=-861.5734, interpMode=RCIM_Cubic, time=0, value=250}, <at 0.2668 150 RCIM_Cubic>]}
       YCurve {keys [{arriveTangent=-66.82, interpMode=RCIM_Cubic, time=0, value=25}, <at 0.2668 5 RCIM_Cubic>]}
       ZCurve {keys [{arriveTangent=-0.0001, interpMode=RCIM_Cubic, time=0, value=1}, <at 0.2668 0 RCIM_Cubic>]}
      }})
    ScaleA=true
    ScaleRGB=false
   )
   &DynamicMaterialParameters(
    Index 0 Param 1=&FloatFromCurve(
     FloatCurve=DataInterface() {propertyValues {Curve {keys [
         <at 0 0 RCIM_Cubic>
         {arriveTangent=1.5947, interpMode=RCIM_Cubic, leaveTangent=1.5947, tangentMode=RCTM_User, time=0.2557, value=0.5257}
         <at 1 1 RCIM_Cubic>
        ]}}}
     CurveIndex=$Particles.Age
     Scale Curve=0.25
    )
    Index 0 Param 3=&FloatFromCurve(FloatCurve=DataInterface() {propertyValues {Curve {keys [<at 0 1>,<at 1 0>]}}})
    Module.Param0WriteEnabled=true
    Module.Param1WriteEnabled=false
    Module.Param2WriteEnabled=true
    Module.Param3WriteEnabled=false
   )
  ]
  rendererClasses [&NiagaraSpriteRendererProperties]
  renderers [{rendererClass=&NiagaraSpriteRendererProperties}]
 }
 Fire_Small {
  emitterUpdateScript [
   &EmitterState()
   &SpawnRate(
    SpawnRate=&FloatFromCurve(
     FloatCurve=DataInterface() {propertyValues {Curve {keys [
         <at 0 1>
         {arriveTangent=0.0423, interpMode=RCIM_Cubic, leaveTangent=0.0423, time=0.153, value=1.1762}
         {arriveTangent=-0.4606, interpMode=RCIM_Cubic, leaveTangent=-0.4606, time=0.2441, value=1.0103}
         <at 0.3053 1.0563>
         {arriveTangent=-0.7586, interpMode=RCIM_Cubic, leaveTangent=-0.7586, time=0.3571, value=1.0822}
         {arriveTangent=-0.9501, interpMode=RCIM_Cubic, leaveTangent=-0.9501, time=0.4956, value=0.8196}
         <at 0.5105 0.8078>
         {arriveTangent=1.381, interpMode=RCIM_Cubic, leaveTangent=1.381, time=0.6074, value=0.8445}
         {arriveTangent=0.3961, interpMode=RCIM_Cubic, leaveTangent=0.3961, time=0.7799, value=1.2121}
         <at 1 1>
        ]}}}
     CurveIndex=$Emitter.NormalizedLoopAge
     Scale Curve=&Lerp_Float(A=0, B=40, Alpha=$System.FN_Integrator.NormalizedValue)
    )
    Use Spawn Probability=false
   )
  ]
  particleSpawnScript [
   &InitializeParticle(
    Lifetime=&RandomRangeFloat(Minimum=0.2, Maximum=0.3, Evaluation Type=::NewEnumerator1)
    Module.Write Color=true
    Module.Write Lifetime=true
    Module.Write Mass=true
    Module.Write Material Random=true
    Module.Write Position=false
    Module.Write Scale=false
    Module.Write SpriteRotation=true
    Module.Write SpriteSize=true
    Sprite Rotation=&NormalizedAngleToDegrees(NormalizedAngle=&RandomRangeFloat(Minimum=0, Maximum=1, Evaluation Type=::NewEnumerator1))
    Sprite Size=&Vector2DFromFloat(Value=&RandomRangeFloat(Minimum=9, Maximum=12, Evaluation Type=::NewEnumerator1))
   )
   &FN_SetSpawnGroup(Number of Spawn Groups=2)
   &Color(Color=&MakeLinearColorFromVectorAndFloat("Vector (RGB)"=<1 0.5 0>, "Float (Alpha)"=&ClampFloat(Float=0.1)))
   &SkeletalMeshLocation(
    Bone / Socket Index=&ModuloInt(A=&ReturnParticleIDAttributes(), B=2)
    Inherit Velocity=false
    MaskSpawnGroup=false
    Mesh Sampling Type=::NewEnumerator5
    Sampled Position Coordinate Space=::Local
    Skeletal Mesh=$User.SKMesh
    Socket Sampling Mode=::NewEnumerator5
   )
  ]
  particleUpdateScript [
   &ParticleState()
   &SkeletalMeshLocation(
    Bone / Socket Index=$Particles.SpawnGroup
    Inherit Velocity=false
    "Inherit Velocity (Scale)"=0
    Interpolate Mesh Sampling=false
    MaskSpawnGroup=false
    Mesh Sampling Type=::NewEnumerator5
    Sampled Orientation Coordinate Space=::Local
    Sampled Position Coordinate Space=::Local
    Sampled Velocity Coordinate Space:ENiagaraCoordinateSpace=::Local
    Skeletal Mesh=$User.SKMesh
    Socket Sampling Mode=::NewEnumerator5
   ) {module=SkeletalMeshLocation001}
   &ScaleSpriteSize(Scale Factor=&Vector2DFromFloat(Value=&FloatFromCurve(
      FloatCurve=DataInterface() {propertyValues {Curve {keys [{arriveTangent=1.4511, interpMode=RCIM_Cubic, time=-0.0013, value=0.1965}, <at 1 1.25>]}}}
      Scale Curve=&Lerp_Float(A=1, B=1.2, Alpha=$System.FN_Integrator.NormalizedValue)
     )))
   &ScaleSpriteSize(Scale Factor=&Vector2DFromFloat(Value=&Lerp_Float(
      A=1
      B=1.25
      Alpha=HlslExpression() {hlslExpression="abs(User.ForwardDot) < .2 &&&& User.RightDot > .2 &&&& Particles.SpawnGroup == 0 ? System.FN_Integrator001.NormalizedValue:0"}
     ))) {module=ScaleSpriteSize001}
   &ScaleSpriteSize(Scale Factor=&Vector2DFromFloat(Value=&Lerp_Float(
      A=1
      B=1.25
      Alpha=HlslExpression() {hlslExpression="abs(User.ForwardDot) < .2 &&&& User.RightDot < -.2 &&&& Particles.SpawnGroup == 1 ? System.FN_Integrator002.NormalizedValue:0"}
     ))) {module=ScaleSpriteSize002}
   &ScaleColor(
    Scale Alpha=&FloatFromCurve(FloatCurve=DataInterface() {propertyValues {Curve {keys [
         <at 0 1 RCIM_Cubic>
         {arriveTangent=-0.573, interpMode=RCIM_Cubic, leaveTangent=-0.573, time=0.5987, value=0.9935}
         {arriveTangent=-2.4759, interpMode=RCIM_Cubic, leaveTangent=-2.4759, time=0.856, value=0.5095}
         <at 1 0>
        ]}}})
    Scale RGB=&VectorFromCurve(VectorCurve=DataInterface() {propertyValues {
       XCurve {keys [{arriveTangent=-861.5734, interpMode=RCIM_Cubic, time=0, value=250}, <at 0.2668 150 RCIM_Cubic>]}
       YCurve {keys [{arriveTangent=-66.82, interpMode=RCIM_Cubic, time=0, value=25}, <at 0.2668 5 RCIM_Cubic>]}
       ZCurve {keys [{arriveTangent=-0.0001, interpMode=RCIM_Cubic, time=0, value=1}, <at 0.2668 0 RCIM_Cubic>]}
      }})
    ScaleA=true
    ScaleRGB=false
   )
   &DynamicMaterialParameters(
    Index 0 Param 1=&FloatFromCurve(
     FloatCurve=DataInterface() {propertyValues {Curve {keys [
         <at 0 0 RCIM_Cubic>
         {arriveTangent=1.5947, interpMode=RCIM_Cubic, leaveTangent=1.5947, tangentMode=RCTM_User, time=0.2557, value=0.5257}
         <at 1 1 RCIM_Cubic>
        ]}}}
     CurveIndex=$Particles.Age
     Scale Curve=0.25
    )
    Index 0 Param 3=&FloatFromCurve(FloatCurve=DataInterface() {propertyValues {Curve {keys [<at 0 1>,<at 1 0.5>]}}})
    Module.Param0WriteEnabled=true
    Module.Param1WriteEnabled=false
    Module.Param2WriteEnabled=true
    Module.Param3WriteEnabled=false
   )
  ]
  rendererClasses [&NiagaraSpriteRendererProperties]
  renderers [{rendererClass=&NiagaraSpriteRendererProperties}]
 }
 FlameBig {
  emitterUpdateScript [
   &EmitterState(Inactive Response=::NewEnumerator1, Life Cycle Mode=::NewEnumerator1, Loop Duration=1, UseLoopDelay=true)
   &SpawnRate(
    Spawn Probability=&Lerp_Float(A=0, B=1, Alpha=$System.FN_Integrator.NormalizedValue)
    SpawnRate=15
    Use Spawn Probability=false
   )
  ]
  particleSpawnScript [
   &InitializeParticle(
    Lifetime=0.3
    Mesh Scale=<0.25 0.25 0.5>
    Module.Write Color=true
    Module.Write Lifetime=true
    Module.Write Mass=true
    Module.Write Material Random=true
    Module.Write Position=false
    Module.Write Scale=true
    Module.Write SpriteRotation=false
    Module.Write SpriteSize=false
    Position=<-30 0 0>
   )
   &Color(Color=&MakeLinearColorFromVectorAndFloat(
     "Vector (RGB)"=&MakeVector(
      X=&Lerp_Float(
       A=1
       B=&FloatFromCurve(
        FloatCurve=DataInterface() {propertyValues {Curve {keys [<at 0 1 RCIM_Cubic>, {arriveTangent=5.5672, interpMode=RCIM_Cubic, leaveTangent=5.5672, tangentMode=RCTM_User, time=0.15, value=3}, <at 1 1>]}}}
        CurveIndex=$System.FN_Integrator.NormalizedValue
       )
       Alpha=&MakeCustomFloatFromBool(Bool=&SetBoolByFloatComparison(A=$User.ForwardDot, B=0), True Float=1, False Float=0)
      )
      Y=&Lerp_Float(
       A=1
       B=&FloatFromCurve(
        FloatCurve=DataInterface() {propertyValues {Curve {keys [<at 0 1 RCIM_Cubic>, {arriveTangent=5.5672, interpMode=RCIM_Cubic, leaveTangent=5.5672, tangentMode=RCTM_User, time=0.15, value=3}, <at 1 1>]}}}
        CurveIndex=$System.FN_Integrator.NormalizedValue
       )
       Alpha=&MakeCustomFloatFromBool(Bool=&SetBoolByFloatComparison(A=$User.ForwardDot, B=0), True Float=1, False Float=0)
      )
      Z=0
     )
     "Float (Alpha)"=&RandomRangeFloat(Minimum=0.5, Maximum=1)
    ))
   &DynamicMaterialParameters(
    Index 0 Param 1=&RandomRangeFloat(Minimum=0.2, Maximum=0.3, Evaluation Type=::NewEnumerator1)
    Module.Param0WriteEnabled=true
    Module.Param1WriteEnabled=false
    Module.Param2WriteEnabled=false
    Module.Param3WriteEnabled=false
   )
   &InitialMeshOrientation(Orientation Axis=<0 0 0>, Orientation Vector=<0 0 0>, Rotation=<0 0.25 0>, Use Orientation Vector=true, Use Rotation Vector=true)
  ]
  particleUpdateScript [
   &ParticleState()
   &SkeletalMeshLocation(
    Apply Sampled Orientation to Mesh Particles=true
    Bone / Socket Index=2
    Inherit Velocity=false
    "Inherit Velocity (Scale)"=0
    Interpolate Mesh Sampling=false
    MaskSpawnGroup=false
    Mesh Sampling Type=::NewEnumerator5
    Sampled Orientation Coordinate Space=::Local
    Sampled Position Coordinate Space=::Local
    Sampled Velocity Coordinate Space:ENiagaraCoordinateSpace=::Local
    Skeletal Mesh=$User.SKMesh
    Socket Sampling Mode=::NewEnumerator5
   )
   &ScaleColor(
    Scale Alpha=&FloatFromCurve(
     FloatCurve=DataInterface() {propertyValues {Curve {keys [{arriveTangent=1, time=0.06, value=0}, <at 0.15 1>, <at 0.2 1>, <at 1 0 RCIM_Cubic>]}}}
     Scale Curve=&Lerp_Float(A=0.25, B=1, Alpha=$System.FN_Integrator.NormalizedValue)
    )
    Scale RGB=&VectorFromFloat(Value=10)
    ScaleA=true
    ScaleRGB=true
   )
   &ScaleMeshSize(Scale Factor=&MakeVector(
     X=&Lerp_Float(
      A=&Lerp_Float(A=0.8, B=1, Alpha=$System.FN_Integrator.NormalizedValue)
      B=&FloatFromCurve(
       FloatCurve=DataInterface() {propertyValues {Curve {keys [
           <at 0 0.2 RCIM_Cubic>
           {arriveTangent=5.5672, interpMode=RCIM_Cubic, leaveTangent=5.5672, tangentMode=RCTM_User, time=0.1585, value=0.7082}
           <at 1 1>
          ]}}}
       CurveIndex=$System.FN_Integrator.NormalizedValue
      )
      Alpha=&MakeCustomFloatFromBool(Bool=&SetBoolByFloatComparison(A=$User.ForwardDot, B=0), True Float=1, False Float=0)
     )
     Y=&Lerp_Float(
      A=&Lerp_Float(A=0.8, B=1, Alpha=$System.FN_Integrator.NormalizedValue)
      B=&FloatFromCurve(
       FloatCurve=DataInterface() {propertyValues {Curve {keys [
           <at 0 0.2 RCIM_Cubic>
           {arriveTangent=5.5672, interpMode=RCIM_Cubic, leaveTangent=5.5672, tangentMode=RCTM_User, time=0.1585, value=0.7082}
           <at 1 1>
          ]}}}
       CurveIndex=$System.FN_Integrator.NormalizedValue
      )
      Alpha=&MakeCustomFloatFromBool(Bool=&SetBoolByFloatComparison(A=$User.ForwardDot, B=0), True Float=1, False Float=0)
     )
     Z=&Lerp_Float(
      A=&Lerp_Float(A=0.2, B=1, Alpha=$System.FN_Integrator.NormalizedValue)
      B=&FloatFromCurve(
       FloatCurve=DataInterface() {propertyValues {Curve {keys [
           <at 0 0.2 RCIM_Cubic>
           {arriveTangent=5.5672, interpMode=RCIM_Cubic, leaveTangent=5.5672, tangentMode=RCTM_User, time=0.1612, value=0.9243}
           <at 1 1>
          ]}}}
       CurveIndex=$System.FN_Integrator.NormalizedValue
      )
      Alpha=&MakeCustomFloatFromBool(Bool=&SetBoolByFloatComparison(A=$User.ForwardDot, B=0), True Float=1, False Float=0)
     )
    ))
   /SetVariables(Particles.Scale_01:Vector3f=$Particles.Scale, Particles.Scale_02:Vector3f=$Particles.Initial.Scale) {isSetParameters=true}
   /SetVariables(Particles.Scale=&Lerp_Vector(
     A=$Particles.Scale_01
     B=$Particles.Initial.Scale
     Alpha=&MakeCustomFloatFromBool(Bool=$User.bisFrontEnd, True Float=1, False Float=0)
    )) {isSetParameters=true}
  ]
  rendererClasses [&NiagaraMeshRendererProperties]
  renderers [{rendererClass=&NiagaraMeshRendererProperties, rendererIndex=0}]
 }
 FlameSmall {
  emitterUpdateScript [
   &EmitterState(Life Cycle Mode=::NewEnumerator1, Loop Duration=1, UseLoopDelay=true)
   &SpawnRate(
    Spawn Probability=&Lerp_Float(A=0, B=1, Alpha=$System.FN_Integrator.NormalizedValue)
    SpawnRate=15
    Use Spawn Probability=false
   )
  ]
  particleSpawnScript [
   &InitializeParticle(
    Lifetime=0.3
    Mesh Scale=<0.1 0.1 0.25>
    Module.Write Color=true
    Module.Write Lifetime=true
    Module.Write Mass=true
    Module.Write Material Random=true
    Module.Write Position=false
    Module.Write Scale=true
    Module.Write SpriteRotation=false
    Module.Write SpriteSize=false
    Position=<-30 0 0>
   )
   &FN_SetSpawnGroup(Number of Spawn Groups=2)
   &Color(Color=&MakeLinearColorFromVectorAndFloat("Vector (RGB)"=<1 1 0>, "Float (Alpha)"=&RandomRangeFloat(Minimum=0.5, Maximum=1)))
   &DynamicMaterialParameters(
    Index 0 Param 1=&RandomRangeFloat(Minimum=0.2, Maximum=0.3, Evaluation Type=::NewEnumerator1)
    Module.Param0WriteEnabled=true
    Module.Param1WriteEnabled=false
    Module.Param2WriteEnabled=false
    Module.Param3WriteEnabled=false
   )
   &InitialMeshOrientation(
    Orientation Axis=<0 0 0>
    Orientation Vector=<0 0 0>
    Rotation=&Lerp_Vector(A=<0 0.25 0>, B=<0 -0.25 0>, Alpha=&MakeFloatFromInt(Int=$Particles.SpawnGroup))
    Use Orientation Vector=true
    Use Rotation Vector=true
   )
  ]
  particleUpdateScript [
   &ParticleState()
   &SkeletalMeshLocation(
    Apply Sampled Orientation to Mesh Particles=true
    Bone / Socket Index=$Particles.SpawnGroup
    Inherit Velocity=false
    "Inherit Velocity (Scale)"=0
    Interpolate Mesh Sampling=false
    MaskSpawnGroup=false
    Mesh Sampling Type=::NewEnumerator5
    Sampled Orientation Coordinate Space=::Local
    Sampled Position Coordinate Space=::Local
    Sampled Velocity Coordinate Space:ENiagaraCoordinateSpace=::Local
    Skeletal Mesh=$User.SKMesh
    Socket Sampling Mode=::NewEnumerator5
   )
   &ScaleColor(
    Scale Alpha=&FloatFromCurve(
     FloatCurve=DataInterface() {propertyValues {Curve {keys [{arriveTangent=1, time=0.06, value=0}, <at 0.15 1>, <at 0.2 1>, <at 1 0 RCIM_Cubic>]}}}
     Scale Curve=&Lerp_Float(A=1, B=1, Alpha=$System.FN_Integrator.NormalizedValue)
    )
    Scale RGB=&VectorFromFloat(Value=10)
    ScaleA=true
    ScaleRGB=true
   )
   &ScaleMeshSize(Scale Factor=&MakeCustomVectorFromBool(
     Bool=HlslExpression() {hlslExpression="Particles.SpawnGroup == 0"}
     True Vector=&Lerp_Vector(
      A=<1 1 0.2>
      B=<1 1 1>
      Alpha=&ReturnLargestFloat(
       A:NiagaraFloat=HlslExpression() {hlslExpression=System.FN_Integrator001.NormalizedValue}
       B:NiagaraFloat=HlslExpression() {hlslExpression=System.FN_Integrator.NormalizedValue}
      )
     )
     False Vector=<1 1 1>
    ))
   &ScaleMeshSize(Scale Factor=&MakeCustomVectorFromBool(
     Bool=HlslExpression() {hlslExpression="Particles.SpawnGroup == 1"}
     True Vector=&Lerp_Vector(
      A=<1 1 0.2>
      B=<1 1 1>
      Alpha=&ReturnLargestFloat(
       A:NiagaraFloat=HlslExpression() {hlslExpression=System.FN_Integrator002.NormalizedValue}
       B:NiagaraFloat=HlslExpression() {hlslExpression=System.FN_Integrator.NormalizedValue}
      )
     )
     False Vector=<1 1 1>
    )) {module=ScaleMeshSize001}
  ]
  rendererClasses [&NiagaraMeshRendererProperties]
  renderers [{rendererClass=&NiagaraMeshRendererProperties, rendererIndex=0}]
 }
 GlowCardBig {
  emitterUpdateScript [
   &EmitterState()
   &SpawnRate(
    SpawnRate=&Lerp_Float(
     A=0
     B=&Lerp_Float(A=6, B=9, Alpha=$System.FN_Integrator.NormalizedValue)
     Alpha=&MakeCustomFloatFromBool(Bool=$User.bisFullyDeployed, True Float=1, False Float=0)
    )
    Use Spawn Probability=false
   )
  ]
  particleSpawnScript [
   &InitializeParticle(
    Lifetime=&Lerp_Float(
     A=&RandomRangeFloat(Minimum=0.45, Maximum=0.55)
     B=&RandomRangeFloat(Minimum=0.3, Maximum=0.4)
     Alpha=$System.FN_Integrator.NormalizedValue
    )
    Module.Write Color=true
    Module.Write Lifetime=true
    Module.Write Mass=true
    Module.Write Material Random=true
    Module.Write Position=false
    Module.Write Scale=false
    Module.Write SpriteRotation=true
    Module.Write SpriteSize=true
    Sprite Rotation=&NormalizedAngleToDegrees(NormalizedAngle=0)
    Sprite Size=<5 5>
   )
   &Color(Color=&MakeLinearColorFromVectorAndFloat(
     "Vector (RGB)"=&Lerp_Vector(A=<7.5 1 0>, B=<12 6 1>, Alpha=$System.FN_Integrator.NormalizedValue)
     "Float (Alpha)"=&ClampFloat(Float=&Lerp_Float(A=0.15, B=0.25, Alpha=$System.FN_Integrator.NormalizedValue))
    ))
   &SkeletalMeshLocation(
    Bone / Socket Index=2
    Inherit Velocity=false
    "Inherit Velocity (Scale)"=0
    Interpolate Mesh Sampling=false
    MaskSpawnGroup=false
    Mesh Sampling Type=::NewEnumerator5
    Sampled Orientation Coordinate Space=::Local
    Sampled Position Coordinate Space=::Local
    Sampled Position Offset Coordinate Space:ENiagaraMeshTransforms=::NewEnumerator2
    Sampled Velocity Coordinate Space:ENiagaraCoordinateSpace=::Local
    Skeletal Mesh=$User.SKMesh
    Socket Sampling Mode=::NewEnumerator5
   )
  ]
  particleUpdateScript [
   &ParticleState()
   &SkeletalMeshLocation(
    Bone / Socket Index=2
    Inherit Velocity=false
    "Inherit Velocity (Scale)"=0
    Interpolate Mesh Sampling=false
    MaskSpawnGroup=false
    Mesh Sampling Type=::NewEnumerator5
    Sampled Orientation Coordinate Space=::Local
    Sampled Position Coordinate Space=::Local
    Sampled Position Offset Coordinate Space:ENiagaraMeshTransforms=::NewEnumerator2
    Sampled Velocity Coordinate Space:ENiagaraCoordinateSpace=::Local
    Skeletal Mesh=$User.SKMesh
    Socket Sampling Mode=::NewEnumerator5
   ) {module=SkeletalMeshLocation001}
   &ScaleColor(
    Scale Alpha=&FloatFromCurve(
     FloatCurve=DataInterface() {propertyValues {Curve {keys [<at 0 0 RCIM_Cubic>,<at 0.5 1 RCIM_Cubic>,<at 1 0 RCIM_Cubic>]}}}
     Scale Curve=&Lerp_Float(A=0.15, B=1, Alpha=$System.FN_Integrator.NormalizedValue)
    )
    Scale RGB=&VectorFromCurve(VectorCurve=DataInterface() {propertyValues {XCurve {keys [<at 0 1>,<at 1 1>]}, YCurve {keys [<at 0 1>,<at 1 0.2>]}, ZCurve {keys [<at 0 1>,<at 1 0.1>]}}})
    ScaleA=true
    ScaleRGB=false
   )
   &ScaleSpriteSize(Scale Factor=&Vector2DFromFloat(Value=&FloatFromCurve(
      FloatCurve=DataInterface() {propertyValues {Curve {keys [
          {arriveTangent=6.5517, interpMode=RCIM_Cubic, time=0, value=1}
          {arriveTangent=6.7766, interpMode=RCIM_Cubic, leaveTangent=6.7766, time=0.0949, value=2.1843}
          {arriveTangent=0.9013, interpMode=RCIM_Cubic, leaveTangent=0.9013, time=0.2688, value=2.8217}
          <at 1 3 RCIM_Cubic>
         ]}}}
      Scale Curve=&Lerp_Float(A=1.2, B=1, Alpha=$System.FN_Integrator.NormalizedValue)
     )))
  ]
  rendererClasses [&NiagaraSpriteRendererProperties]
  renderers [{rendererClass=&NiagaraSpriteRendererProperties}]
 }
 GlowCard_Small {
  emitterUpdateScript [
   &EmitterState(
    Life Cycle Mode=::NewEnumerator1
    Loop Duration=&RandomRangeFloat(Minimum=0.15, Maximum=0.25)
    Recalculate Duration Each Loop=true
    UseLoopDelay=false
   )
   &SpawnBurst_Instantaneous(
    Spawn Count=4
    Spawn Probability=&Lerp_Float(
     A=0.2
     B=1
     Alpha=&ReturnLargestFloat(
      A:NiagaraFloat=$System.FN_Integrator.NormalizedValue
      B:NiagaraFloat=&ReturnLargestFloat(A:NiagaraFloat=$System.FN_Integrator002.NormalizedValue, B:NiagaraFloat=$System.FN_Integrator001.NormalizedValue)
     )
    )
    Use Spawn Probability=true
   )
  ]
  particleSpawnScript [
   &InitializeParticle(
    Lifetime=&RandomRangeFloat(Minimum=0.1, Maximum=0.15)
    Module.Write Color=true
    Module.Write Lifetime=true
    Module.Write Mass=true
    Module.Write Material Random=true
    Module.Write Position=false
    Module.Write Scale=false
    Module.Write SpriteRotation=true
    Module.Write SpriteSize=true
    Sprite Rotation=&NormalizedAngleToDegrees(NormalizedAngle=0)
    Sprite Size=<2.5 2.5>
   )
   &FN_SetSpawnGroup(Number of Spawn Groups=2)
   &Color(Color=&MakeLinearColorFromVectorAndFloat("Vector (RGB)"=<10 2.5 0>, "Float (Alpha)"=&ClampFloat(Float=0.25)))
   &SkeletalMeshLocation(
    Bone / Socket Index=$Particles.SpawnGroup
    Inherit Velocity=false
    "Inherit Velocity (Scale)"=0
    Interpolate Mesh Sampling=false
    MaskSpawnGroup=false
    Mesh Sampling Type=::NewEnumerator5
    Sampled Orientation Coordinate Space=::Local
    Sampled Position Coordinate Space=::Local
    Sampled Velocity Coordinate Space:ENiagaraCoordinateSpace=::Local
    Skeletal Mesh=$User.SKMesh
    Socket Sampling Mode=::NewEnumerator5
   )
  ]
  particleUpdateScript [
   &ParticleState()
   &SkeletalMeshLocation(
    Bone / Socket Index=$Particles.SpawnGroup
    Inherit Velocity=false
    "Inherit Velocity (Scale)"=0
    Interpolate Mesh Sampling=false
    MaskSpawnGroup=false
    Mesh Sampling Type=::NewEnumerator5
    Sampled Orientation Coordinate Space=::Local
    Sampled Position Coordinate Space=::Local
    Sampled Velocity Coordinate Space:ENiagaraCoordinateSpace=::Local
    Skeletal Mesh=$User.SKMesh
    Socket Sampling Mode=::NewEnumerator5
   ) {module=SkeletalMeshLocation001}
   &ScaleColor(
    Scale Alpha=&FloatFromCurve(FloatCurve=DataInterface() {propertyValues {Curve {keys [<at 0 0 RCIM_Cubic>,<at 0.5 1 RCIM_Cubic>,<at 1 0 RCIM_Cubic>]}}})
    Scale RGB=&VectorFromCurve(VectorCurve=DataInterface() {propertyValues {XCurve {keys [<at 0 1>,<at 1 1>]}, YCurve {keys [<at 0 1>,<at 1 0.2>]}, ZCurve {keys [<at 0 1>,<at 1 0.1>]}}})
    ScaleA=true
    ScaleRGB=false
   )
   &ScaleSpriteSize(Scale Factor=&Vector2DFromFloat(Value=&FloatFromCurve(
      FloatCurve=DataInterface() {propertyValues {Curve {keys [
          {arriveTangent=6.5517, interpMode=RCIM_Cubic, time=0, value=1}
          {arriveTangent=6.7766, interpMode=RCIM_Cubic, leaveTangent=6.7766, time=0.0949, value=2.1843}
          {arriveTangent=0.9013, interpMode=RCIM_Cubic, leaveTangent=0.9013, time=0.2688, value=2.8217}
          <at 1 3 RCIM_Cubic>
         ]}}}
      Scale Curve=&Lerp_Float(A=1, B=1.2, Alpha=$System.FN_Integrator.NormalizedValue)
     )))
   &ScaleSpriteSize(Scale Factor=&Vector2DFromFloat(Value=&Lerp_Float(
      A=1
      B=1.25
      Alpha=HlslExpression() {hlslExpression="abs(User.ForwardDot) < .2 &&&& User.RightDot > .2 &&&& Particles.SpawnGroup == 0 ? System.FN_Integrator001.NormalizedValue:0"}
     ))) {module=ScaleSpriteSize001}
   &ScaleSpriteSize(Scale Factor=&Vector2DFromFloat(Value=&Lerp_Float(
      A=1
      B=1.25
      Alpha=HlslExpression() {hlslExpression="abs(User.ForwardDot) < .2 &&&& User.RightDot < -.2 &&&& Particles.SpawnGroup == 1 ? System.FN_Integrator002.NormalizedValue:0"}
     ))) {module=ScaleSpriteSize002}
  ]
  rendererClasses [&NiagaraSpriteRendererProperties]
  renderers [{rendererClass=&NiagaraSpriteRendererProperties}]
 }
 LightBig {
  emitterUpdateScript [
   &EmitterState()
   &SpawnBurst_Instantaneous(Spawn Count=1, Spawn Time=0.01, Use Spawn Probability=false) {module=SpawnBurst_Instantaneous001}
  ]
  particleSpawnScript [&InitializeParticle(
    Module.Write Color=true
    Module.Write Lifetime=true
    Module.Write Mass=true
    Module.Write Material Random=true
    Module.Write Position=false
    Module.Write Scale=false
    Module.Write SpriteRotation=false
    Module.Write SpriteSize=true
    Position=<-8 0 0>
    Sprite Size=<1.25 1.25>
   )]
  particleUpdateScript [
   &ParticleState()
   &SkeletalMeshLocation(
    Bone / Socket Index=2
    Inherit Velocity=false
    MaskSpawnGroup=false
    Mesh Sampling Type=::NewEnumerator5
    Sampled Orientation Coordinate Space=::Local
    Sampled Position Coordinate Space=::Local
    Sampled Position Offset Coordinate Space:ENiagaraMeshTransforms=::NewEnumerator2
    Sampled Velocity Coordinate Space:ENiagaraCoordinateSpace=::Local
    Skeletal Mesh=$User.SKMesh
    Socket Sampling Mode=::NewEnumerator5
   )
   &FN_Light_Attributes(
    Color=<rgba 2 0.5 0.025 1>
    Create Color Attribute=true
    Create Exponent Attribute=true
    Create Position Attribute=true
    Create Radius Attribute=true
    Enable Particle Light Enable=true
    Exponent=5
    Light Enabled=true
    Position=&Add_Vector(A=Linked() {linkedVariable {name=Particles.Position, type=NiagaraPosition}}, B=<0 18 0>)
    Radius=150
    Scale Color=&FloatFromCurve(
     FloatCurve=DataInterface() {propertyValues {Curve {keys [
         <at 0 1 RCIM_Cubic>
         {arriveTangent=-0.5591, interpMode=RCIM_Cubic, leaveTangent=-0.5591, time=0.1488, value=1.0253}
         {arriveTangent=0.4628, interpMode=RCIM_Cubic, leaveTangent=0.4628, time=0.2431, value=0.8641}
         <at 0.3182 1.1037>
         {arriveTangent=-0.8449, interpMode=RCIM_Cubic, leaveTangent=-0.8449, time=0.3716, value=0.9639}
         {arriveTangent=-0.6878, interpMode=RCIM_Cubic, leaveTangent=-0.6878, time=0.4572, value=0.9863}
         <at 0.5056 0.8717>
         {arriveTangent=1.5338, interpMode=RCIM_Cubic, leaveTangent=1.5338, time=0.5767, value=1.1763}
         {arriveTangent=2.8113, interpMode=RCIM_Cubic, leaveTangent=2.8113, tangentMode=RCTM_User, time=0.6624, value=1.1121}
         {arriveTangent=-0.3543, interpMode=RCIM_Cubic, leaveTangent=-0.3543, time=0.7641, value=0.8773}
         <at 0.9786 1 RCIM_Cubic>
        ]}}}
     Scale Curve=&Lerp_Float(A=3, B=9, Alpha=$System.FN_Integrator.NormalizedValue)
    )
   )
   &FN_Gen5_Light(
    Attenutation=10
    Color:LinearColor=<rgba 1 0.5 0 1>
    Intensity:NiagaraFloat=&FloatFromCurve(
     FloatCurve=DataInterface() {propertyValues {Curve {keys [
         <at 0 1 RCIM_Cubic>
         {arriveTangent=-1.5889, interpMode=RCIM_Cubic, leaveTangent=-1.5889, time=0.1488, value=1.0253}
         {arriveTangent=3.5922, interpMode=RCIM_Cubic, leaveTangent=3.5922, time=0.2467, value=0.6081}
         <at 0.3244 1.6559>
         {arriveTangent=-5.0409, interpMode=RCIM_Cubic, leaveTangent=-5.0409, time=0.3716, value=0.9639}
         {arriveTangent=-2.0797, interpMode=RCIM_Cubic, leaveTangent=-2.0797, time=0.4572, value=0.9863}
         <at 0.5074 0.6815>
         {arriveTangent=6.1003, interpMode=RCIM_Cubic, leaveTangent=6.1003, time=0.5828, value=1.7285}
         {arriveTangent=2.8113, interpMode=RCIM_Cubic, leaveTangent=2.8113, tangentMode=RCTM_User, time=0.6685, value=1.6643}
         {arriveTangent=-2.1415, interpMode=RCIM_Cubic, leaveTangent=-2.1415, time=0.7658, value=0.6871}
         <at 0.9786 1 RCIM_Cubic>
        ]}}}
     Scale Curve=&Lerp_Float(A=350, B=1000, Alpha=$System.FN_Integrator.NormalizedValue)
    )
    Scale:Vector3f=&VectorFromFloat(Value=1)
   )
  ]
  rendererClasses [&NiagaraLightRendererProperties, &NiagaraComponentRendererProperties]
  renderers [
   {rendererClass=&NiagaraLightRendererProperties, rendererIndex=0}
   {rendererClass=&NiagaraComponentRendererProperties, rendererIndex=1}
  ]
 }
 LightSmall {
  emitterUpdateScript [
   &EmitterState()
   &SpawnBurst_Instantaneous(Spawn Count=4, Use Spawn Probability=false) {module=SpawnBurst_Instantaneous001}
  ]
  particleSpawnScript [
   &InitializeParticle(
    Module.Write Color=true
    Module.Write Lifetime=true
    Module.Write Mass=true
    Module.Write Material Random=true
    Module.Write Position=false
    Module.Write Scale=false
    Module.Write SpriteRotation=false
    Module.Write SpriteSize=true
    Position=<-8 0 0>
    Sprite Size=<20 20>
   )
   &FN_SetSpawnGroup(Number of Spawn Groups=2)
  ]
  particleUpdateScript [
   &ParticleState()
   &SkeletalMeshLocation(
    Bone / Socket Index=$Particles.SpawnGroup
    Inherit Velocity=false
    MaskSpawnGroup=false
    Mesh Sampling Type=::NewEnumerator5
    Sampled Orientation Coordinate Space=::Local
    Sampled Position Coordinate Space=::Local
    Sampled Position Offset Coordinate Space:ENiagaraMeshTransforms=::NewEnumerator2
    Sampled Velocity Coordinate Space:ENiagaraCoordinateSpace=::Local
    Skeletal Mesh=$User.SKMesh
    Socket Sampling Mode=::NewEnumerator5
   )
   &FN_Gen5_Light(
    Attenutation=6.5
    Color:LinearColor=<rgba 1 0.5 0 1>
    Exponent=2
    Intensity:NiagaraFloat=&FloatFromCurve(
     FloatCurve=DataInterface() {propertyValues {Curve {keys [
         <at 0 1 RCIM_Cubic>
         {arriveTangent=-2.3439, interpMode=RCIM_Cubic, leaveTangent=-2.3439, time=0.1488, value=1.0253}
         {arriveTangent=3.5923, interpMode=RCIM_Cubic, leaveTangent=3.5923, time=0.2377, value=0.4429}
         <at 0.3244 1.6559>
         {arriveTangent=-5.0409, interpMode=RCIM_Cubic, leaveTangent=-5.0409, time=0.3716, value=0.9639}
         {arriveTangent=-3.5299, interpMode=RCIM_Cubic, leaveTangent=-3.5299, time=0.4572, value=0.9863}
         <at 0.4983 0.5164>
         {arriveTangent=6.7478, interpMode=RCIM_Cubic, leaveTangent=6.7478, time=0.5828, value=1.7285}
         {arriveTangent=2.8113, interpMode=RCIM_Cubic, leaveTangent=2.8113, tangentMode=RCTM_User, time=0.6685, value=1.6643}
         {arriveTangent=-2.1415, interpMode=RCIM_Cubic, leaveTangent=-2.1415, time=0.7568, value=0.522}
         <at 0.9786 1 RCIM_Cubic>
        ]}}}
     Scale Curve=&Lerp_Float(A=50, B=250, Alpha=$System.FN_Integrator.NormalizedValue)
    )
    Scale:Vector3f=&VectorFromFloat(Value=1)
   )
  ]
  rendererClasses [&NiagaraComponentRendererProperties]
  renderers [{rendererClass=&NiagaraComponentRendererProperties, rendererIndex=0}]
 }
 SoftJetBig {
  emitterUpdateScript [
   &EmitterState()
   &SpawnRate(
    Spawn Probability=&Lerp_Float(
     A=&Lerp_Float(
      A=&Lerp_Float(A=0, B=1, Alpha=$System.FN_Integrator.NormalizedValue)
      B=&FloatFromCurve(
       FloatCurve=DataInterface() {propertyValues {Curve {keys [
           <at 0 0.2 RCIM_Cubic>
           {arriveTangent=5.5672, interpMode=RCIM_Cubic, leaveTangent=5.5672, tangentMode=RCTM_User, time=0.15, value=2}
           <at 1 1>
          ]}}}
       CurveIndex=$System.FN_Integrator.NormalizedValue
      )
      Alpha=&MakeCustomFloatFromBool(Bool=&SetBoolByFloatComparison(A=$User.ForwardDot, B=0), True Float=1, False Float=0)
     )
     B=1
     Alpha=&MakeCustomFloatFromBool(Bool=$User.bisFrontEnd, True Float=1, False Float=0)
    )
    SpawnRate=50
    Use Spawn Probability=true
   )
  ]
  particleSpawnScript [
   &InitializeParticle(
    Color=<rgba 1.5 1 0.3 1>
    Lifetime=0.1
    Module.Write Color=true
    Module.Write Lifetime=true
    Module.Write Mass=true
    Module.Write Material Random=true
    Module.Write Position=false
    Module.Write Scale=false
    Module.Write SpriteRotation=false
    Module.Write SpriteSize=true
    Position=<0 -1 0>
    Sprite Size=&RandomRangeVector2D(Minimum=<15 70>, Maximum=<20 100>, Evaluation Type=::NewEnumerator1)
   )
   &SkeletalMeshLocation(
    Bone / Socket Index=2
    Inherit Velocity=false
    MaskSpawnGroup=false
    Mesh Sampling Type=::NewEnumerator5
    Sampled Position Coordinate Space=::Local
    Sampled Position Offset=<0 0 0.14>
    Skeletal Mesh=$User.SKMesh
    Socket Sampling Mode=::NewEnumerator5
   )
   &AddVelocity(Velocity=&RotateVectorByQuaternion(Vector To Rotate=<75 0 0>, Quaternion=$Output.SkeletalMeshLocation.SampledRotation))
   &Color(Color=&MakeLinearColorFromVectorAndFloat("Vector (RGB)"=<5 5 5>, "Float (Alpha)"=&ClampFloat(Float=1)))
  ]
  particleUpdateScript [
   &ParticleState()
   &SolveForcesAndVelocity(Clamp Velocity=false, Limit Acceleration=false)
   &ScaleColor(
    Scale Alpha=&FloatFromCurve(FloatCurve=DataInterface() {propertyValues {Curve {keys [<at 0 0 RCIM_Cubic>,<at 0.25 1 RCIM_Cubic>,<at 1 0 RCIM_Cubic>]}}})
    Scale RGB=&VectorFromCurve(VectorCurve=DataInterface() {propertyValues {
       XCurve {keys [<at 0 10 RCIM_Cubic>,<at 1 1>]}
       YCurve {keys [<at 0 1 RCIM_Cubic>,<at 1 0.4798>]}
       ZCurve {keys [<at 0 0 RCIM_Cubic>,<at 1 0>]}
      }})
    ScaleA=true
    ScaleRGB=true
   )
   &ScaleSpriteSize(Scale Factor=&Vector2DFromCurve(
     Vector2Curve=DataInterface() {propertyValues {XCurve {keys [<at 0 1.5>,<at 1 0.25>]}, YCurve {keys [<at 0 1>,<at 1 1.75>]}}}
     ScaleCurve=<0.5 1>
    ))
   &ScaleSpriteSize(Scale Factor=&Vector2DFromFloat(Value=&Lerp_Float(
      A=&Lerp_Float(A=0, B=1, Alpha=$System.FN_Integrator.NormalizedValue)
      B=&FloatFromCurve(
       FloatCurve=DataInterface() {propertyValues {Curve {keys [
           <at 0 0.2 RCIM_Cubic>
           {arriveTangent=5.5672, interpMode=RCIM_Cubic, leaveTangent=5.5672, tangentMode=RCTM_User, time=0.1748, value=0.9401}
           <at 1 1>
          ]}}}
       CurveIndex=$System.FN_Integrator.NormalizedValue
      )
      Alpha=&MakeCustomFloatFromBool(Bool=&SetBoolByFloatComparison(A=$User.ForwardDot, B=0), True Float=1, False Float=0)
     ))) {module=ScaleSpriteSize001}
   &ScaleSpriteSize(Scale Factor=&MakeVector2D(
     X=&Lerp_Float(
      A=&Lerp_Float(A=1, B=1, Alpha=$System.FN_Integrator.NormalizedValue)
      B=&FloatFromCurve(
       FloatCurve=DataInterface() {propertyValues {Curve {keys [<at 0 1 RCIM_Cubic>,<at 0.25 1.25>,<at 1 1>]}}}
       CurveIndex=$System.FN_Integrator.NormalizedValue
      )
      Alpha=&MakeCustomFloatFromBool(Bool=&SetBoolByFloatComparison(A=$User.ForwardDot, B=0), True Float=1, False Float=0)
     )
     Y=1
    )) {module=ScaleSpriteSize002}
   /SetVariables(Particles.SpriteSize=&Lerp_Vector2(
     A=$Particles.SpriteSize
     B=&Multiply_Vector2(
      A=&RandomRangeVector2D(Minimum=<15 30>, Maximum=<20 40>)
      B=&Vector2DFromCurve(
       Vector2Curve=DataInterface() {propertyValues {XCurve {keys [<at 0 1.5>,<at 1 0.25>]}, YCurve {keys [<at 0 1>,<at 1 1.75>]}}}
       ScaleCurve=<0.5 1>
      )
     )
     Alpha=&MakeCustomFloatFromBool(Bool=$User.bisFrontEnd, True Float=1, False Float=0)
    )) {isSetParameters=true}
  ]
  rendererClasses [&NiagaraSpriteRendererProperties]
  renderers [{rendererClass=&NiagaraSpriteRendererProperties}]
 }
 SoftJetSmall {
  emitterUpdateScript [
   &EmitterState()
   &SpawnRate(
    Spawn Probability=&Lerp_Float(
     A=0
     B=1
     Alpha=&ReturnLargestFloat(
      A:NiagaraFloat=$System.FN_Integrator.NormalizedValue
      B:NiagaraFloat=&ReturnLargestFloat(A:NiagaraFloat=$System.FN_Integrator002.NormalizedValue, B:NiagaraFloat=$System.FN_Integrator001.NormalizedValue)
     )
    )
    SpawnRate=100
    Use Spawn Probability=true
   )
  ]
  particleSpawnScript [
   &InitializeParticle(
    Color=<rgba 1.5 1 0.3 1>
    Lifetime=0.1
    Module.Write Color=true
    Module.Write Lifetime=true
    Module.Write Mass=true
    Module.Write Material Random=true
    Module.Write Position=false
    Module.Write Scale=false
    Module.Write SpriteRotation=false
    Module.Write SpriteSize=true
    Position=<0 -1 0>
    Sprite Size=&RandomRangeVector2D(Minimum=<10 15>, Maximum=<15 20>, Evaluation Type=::NewEnumerator1)
   )
   &FN_SetSpawnGroup(Number of Spawn Groups=2)
   &SkeletalMeshLocation(
    Bone / Socket Index=$Particles.SpawnGroup
    Inherit Velocity=false
    MaskSpawnGroup=false
    Mesh Sampling Type=::NewEnumerator5
    Sampled Position Coordinate Space=::Local
    Sampled Position Offset=<0 0 0.14>
    Skeletal Mesh=$User.SKMesh
    Socket Sampling Mode=::NewEnumerator5
   )
   &Color(Color=&MakeLinearColorFromVectorAndFloat("Vector (RGB)"=<5 5 5>, "Float (Alpha)"=&ClampFloat(Float=1)))
   &AddVelocity(Velocity=&RotateVectorByQuaternion(
     Vector To Rotate=&MakeVector(
      X=&Lerp_Float(A=75, B=-75, Alpha=&MakeFloatFromInt(Int=$Particles.SpawnGroup))
      Y=0
      Z=0
     )
     Quaternion=$Output.SkeletalMeshLocation.SampledRotation
    ))
  ]
  particleUpdateScript [
   &ParticleState()
   &SolveForcesAndVelocity(Clamp Velocity=false, Limit Acceleration=false)
   &ScaleColor(
    Scale Alpha=&FloatFromCurve(FloatCurve=DataInterface() {propertyValues {Curve {keys [<at 0 0 RCIM_Cubic>,<at 0.25 1 RCIM_Cubic>,<at 1 0 RCIM_Cubic>]}}})
    Scale RGB=&VectorFromCurve(VectorCurve=DataInterface() {propertyValues {
       XCurve {keys [<at 0 10 RCIM_Cubic>,<at 1 1>]}
       YCurve {keys [<at 0 1 RCIM_Cubic>,<at 1 0.4798>]}
       ZCurve {keys [<at 0 0 RCIM_Cubic>,<at 1 0>]}
      }})
    ScaleA=true
    ScaleRGB=true
   )
   &ScaleSpriteSize(Scale Factor=&Vector2DFromCurve(
     Vector2Curve=DataInterface() {propertyValues {XCurve {keys [<at 0 1.5>,<at 1 0.25>]}, YCurve {keys [<at 0 1>,<at 1 1.75>]}}}
     ScaleCurve=<0.5 1>
    ))
   &ScaleSpriteSize(Scale Factor=&Vector2DFromFloat(Value=&MakeCustomFloatFromBool(
      Bool=HlslExpression() {hlslExpression="Particles.SpawnGroup == 0"}
      True Float=&Lerp_Float(
       A=0
       B=1
       Alpha=&ReturnLargestFloat(A:NiagaraFloat=$System.FN_Integrator001.NormalizedValue, B:NiagaraFloat=$System.FN_Integrator.NormalizedValue)
      )
      False Float=1
     ))) {module=ScaleSpriteSize001}
   &ScaleSpriteSize(Scale Factor=&Vector2DFromFloat(Value=&MakeCustomFloatFromBool(
      Bool=HlslExpression() {hlslExpression="Particles.SpawnGroup == 1"}
      True Float=&Lerp_Float(
       A=0
       B=1
       Alpha=&ReturnLargestFloat(A:NiagaraFloat=$System.FN_Integrator002.NormalizedValue, B:NiagaraFloat=$System.FN_Integrator.NormalizedValue)
      )
      False Float=1
     ))) {module=ScaleSpriteSize002}
  ]
  rendererClasses [&NiagaraSpriteRendererProperties]
  renderers [{rendererClass=&NiagaraSpriteRendererProperties}]
 }
}
summary {
 systemName=NS_Glider_SpaceWarrior_M_Jet
 emitters [
  {emitterName=GlowCardBig, rendererClasses [&NiagaraSpriteRendererProperties]}
  {emitterName=FireBig, rendererClasses [&NiagaraSpriteRendererProperties]}
  {emitterName=LightBig, rendererClasses [&NiagaraLightRendererProperties, &NiagaraComponentRendererProperties]}
  {emitterName=GlowCard_Small, rendererClasses [&NiagaraSpriteRendererProperties]}
  {emitterName=Fire_Small, rendererClasses [&NiagaraSpriteRendererProperties]}
  {emitterName=Airflow_Down, rendererClasses [&NiagaraMeshRendererProperties]}
  {emitterName=FlameBig, rendererClasses [&NiagaraMeshRendererProperties]}
  {emitterName=SoftJetBig, rendererClasses [&NiagaraSpriteRendererProperties]}
  {emitterName=LightSmall, rendererClasses [&NiagaraComponentRendererProperties]}
  {emitterName=FlameSmall, rendererClasses [&NiagaraMeshRendererProperties]}
  {emitterName=SoftJetSmall, rendererClasses [&NiagaraSpriteRendererProperties]}
  {emitterName=Airflow_MotionLines, rendererClasses [&NiagaraSpriteRendererProperties]}
  {emitterName=Airflow_Back, rendererClasses [&NiagaraMeshRendererProperties]}
  {emitterName=Airflow_MotionLines001, rendererClasses [&NiagaraSpriteRendererProperties]}
 ]
 userVariables [
  User.bisFrontEnd {type=NiagaraBool}
  User.bisFullyDeployed {type=NiagaraBool}
  User.ForwardDot {type=NiagaraFloat, defaultValue=0, description=""}
  User.RightDot {type=NiagaraFloat, defaultValue=0, description=""}
  User.SKMesh {
   type=NiagaraDataInterfaceSkeletalMesh
   defaultValue {
    struct=NiagaraExt_VariableValue_DataInterface
    value {dataInterface=&self:NiagaraDataInterfaceSkeletalMesh_0}
   }
  }
 ]
}
